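# Animation Blueprint for the MasalaRail backpack cosmetic, playing idle and reactive animations.
# Blends between an in-game state machine (chest open, damage, elimination, driven by anim input events) and a timer-driven frontend one reusing the elimination animation.
# Additive bone transforms then tilt the pack for skydive/parachute states and offset it using a correction mapped from spine yaw.
# An AnimDynamics chain for arms, skull and eyes exists but is disconnected from the output.
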AnimBlueprint /BRCosmetics/Accessories/FORT_Backpacks/Backpack_F_MED_MasalaRail/Meshes/F_MED_MasalaRail_Pack_AnimBP
target_skeleton=&/F_MED_MasalaRail_Pack_Skeleton
graphs [
 AnimGraph {nodes [
   n1 = Anim Dynamics - Bone: dyn_upperarm_r(Alpha=1, ExternalForce="0,0,0") {
    angularDampingOverride=0.999
    angularSpringConstant=300
    bAngularSpring=true
    bOverrideAngularDamping=true
    bUseGravityOverride=true
    simulationSpace=RootRelative
    boundBone {boneName=dyn_upperarm_r}
    componentAppliedLinearAccClamp=<10000 10000 10000>
    componentLinearAccScale=<0.3 0.3 0.3>
    componentLinearVelScale=<0.2 0.2 0.1>
    gravityOverride {z=-800}
    physicsBodyDefinitions [{
      boundBone {boneName=dyn_upperarm_r}
      boxExtents=<10 10 10>
      constraintSetup {angularLimitsMax=<8 10 2>, angularLimitsMin=<-30 -1 -30>, angularTarget=<1 0 0>}
      localJointOffset=<-1 0 0>
     }]
   }
   n2 = Anim Dynamics - Bone: dyn_lowerarm_r(ComponentPose=@n1, Alpha=1, ExternalForce="0,0,0") {
    angularDampingOverride=0.999
    angularSpringConstant=600
    bAngularSpring=true
    bOverrideAngularDamping=true
    bUseGravityOverride=true
    simulationSpace=RootRelative
    boundBone {boneName=dyn_lowerarm_r}
    componentAppliedLinearAccClamp=<10000 10000 10000>
    componentLinearAccScale=<0.2 0.2 0.2>
    componentLinearVelScale=<0.2 0.2 0.1>
    gravityOverride {z=-700}
    physicsBodyDefinitions [{
      boundBone {boneName=dyn_lowerarm_r}
      boxExtents=<12 12 12>
      constraintSetup {angularLimitsMax=<8 0 5>, angularLimitsMin=<-15 -15 -30>, angularTarget=<1 0 0>}
      localJointOffset=<-0.5 0 0>
     }]
   }
   n3 = Anim Dynamics - Bone: dyn_upperarm_l(ComponentPose=@n2, Alpha=1, ExternalForce="0,0,0") {
    angularDampingOverride=0.999
    angularSpringConstant=300
    bAngularSpring=true
    bOverrideAngularDamping=true
    bUseGravityOverride=true
    simulationSpace=RootRelative
    boundBone {boneName=dyn_upperarm_l}
    componentAppliedLinearAccClamp=<10000 10000 10000>
    componentLinearAccScale=<0.3 0.3 0.3>
    componentLinearVelScale=<0.2 0.2 0.1>
    gravityOverride {z=-800}
    physicsBodyDefinitions [{
      boundBone {boneName=dyn_upperarm_l}
      boxExtents=<10 10 10>
      constraintSetup {angularLimitsMax=<8 10 2>, angularLimitsMin=<-30 -1 -30>, angularTarget=<1 0 0>}
      localJointOffset=<1 0 0>
     }]
   }
   n4 = Anim Dynamics - Bone: dyn_lowerarm_l(ComponentPose=@n3, Alpha=1, ExternalForce="0,0,0") {
    angularDampingOverride=1
    angularSpringConstant=3000
    bAngularSpring=true
    bOverrideAngularDamping=true
    bUseGravityOverride=true
    simulationSpace=RootRelative
    boundBone {boneName=dyn_lowerarm_l}
    componentAppliedLinearAccClamp=<10000 10000 10000>
    componentLinearAccScale=<0.3 0.3 0.3>
    componentLinearVelScale=<0.2 0.2 0.1>
    gravityOverride {z=-900}
    physicsBodyDefinitions [{
      boundBone {boneName=dyn_lowerarm_l}
      boxExtents=<10 10 10>
      constraintSetup {angularLimitsMax=<8 35 50>, angularLimitsMin=<-5 -5 -30>, angularTarget=<1 0 0>}
      localJointOffset=<1 0 0>
     }]
   }
   n5 = Anim Dynamics - Bone: dyn_skull(ComponentPose=@n4, Alpha=1, ExternalForce="0,0,0") {
    angularDampingOverride=0.999
    angularSpringConstant=1000
    bAngularSpring=true
    bOverrideAngularDamping=true
    bUseGravityOverride=true
    simulationSpace=RootRelative
    boundBone {boneName=dyn_skull}
    componentAppliedLinearAccClamp=<10000 10000 10000>
    componentLinearAccScale=<0.3 0.3 0.3>
    componentLinearVelScale=<0.2 0.2 0.1>
    gravityOverride {z=-900}
    physicsBodyDefinitions [{
      boundBone {boneName=dyn_skull}
      boxExtents=<12 12 12>
      constraintSetup {angularLimitsMax=<0 2 1>, angularLimitsMin=<0 -20 -1>, angularTarget=<1 0 0>}
      localJointOffset=<0.2 0 0>
     }]
   }
   n6 = Anim Dynamics - Bone: dyn_eye_upper_l(ComponentPose=@n5, Alpha=1, ExternalForce="0,0,0") {
    angularDampingOverride=0.999
    angularSpringConstant=1000
    bAngularSpring=true
    bOverrideAngularDamping=true
    bUseGravityOverride=true
    simulationSpace=RootRelative
    boundBone {boneName=dyn_eye_upper_l}
    componentAppliedLinearAccClamp=<10000 10000 10000>
    componentLinearAccScale=<0.3 0.3 0.3>
    componentLinearVelScale=<0.2 0.2 0.1>
    gravityOverride {z=-500}
    physicsBodyDefinitions [{
      boundBone {boneName=dyn_eye_upper_l}
      boxExtents=<12 12 12>
      constraintSetup {angularLimitsMax=<0 15 0>, angularLimitsMin=<0 -10 0>, angularTarget=<1 0 0>}
      localJointOffset=<0.5 0 0>
     }]
   }
   Anim Dynamics - Bone: dyn_eye_upper_r(ComponentPose=@n6, Alpha=1, ExternalForce="0,0,0") {
    angularDampingOverride=0.999
    angularSpringConstant=1000
    bAngularSpring=true
    bOverrideAngularDamping=true
    bUseGravityOverride=true
    simulationSpace=RootRelative
    boundBone {boneName=dyn_eye_upper_r}
    componentAppliedLinearAccClamp=<10000 10000 10000>
    componentLinearAccScale=<0.3 0.3 0.3>
    componentLinearVelScale=<0.2 0.2 0.1>
    gravityOverride {z=-500}
    physicsBodyDefinitions [{
      boundBone {boneName=dyn_eye_upper_r}
      boxExtents=<12 12 12>
      constraintSetup {angularLimitsMax=<0 15 0>, angularLimitsMin=<0 -10 0>, angularTarget=<1 0 0>}
      localJointOffset=<-0.5 0 0>
     }]
   }
   Input Pose()
   AnimGraphNode_StateMachine_0 = MasalaRail_StateMachine() {is=title AnimGraphNode_StateMachine}
   AnimGraphNode_StateMachine_1 = FE_Reactivity() {is=title AnimGraphNode_StateMachine}
   n7 = Get bShouldDisplayFrontendReactivity() -> (bShouldDisplayFrontendReactivity:bool)
   n8 = "Transform (Modify) Bone - Bone: dyn_all_pack"(
    ComponentPose=Local To Component(LocalPose=Blend Poses by bool(BlendPose_0=@AnimGraphNode_StateMachine_1, BlendPose_1=@AnimGraphNode_StateMachine_0, bActiveValue=@n7) {blendType=Linear})
    Alpha=Get Alpha_skydiveMode() -> (Alpha_skydiveMode:double)
    Translation="0,0,0"
    Rotation="10,0,0"
    Scale="1,1,1"
   ) {
    rotationMode=BMM_Additive
    rotationSpace=BCS_BoneSpace
    scaleMode=BMM_Additive
    scaleSpace=BCS_BoneSpace
    translationMode=BMM_Additive
    translationSpace=BCS_BoneSpace
    alphaBoolBlend {blendOption=Linear}
    boneToModify {boneName=dyn_all_pack}
    rotation {pitch=10}
   }
   n9 = "Transform (Modify) Bone - Bone: dyn_all_pack"(
    ComponentPose=@n8
    Alpha=Get Parachute _correction() -> (Parachute _correction:double)
    Translation="13,0,-3"
    Rotation="60,0,0"
    Scale="1,1,1"
   ) {
    rotationMode=BMM_Additive
    rotationSpace=BCS_BoneSpace
    translationMode=BMM_Additive
    translationSpace=BCS_BoneSpace
    alphaBoolBlend {blendOption=Linear}
    boneToModify {boneName=dyn_all_pack}
    rotation {pitch=60}
    translation {x=13, z=-3}
   }
   Output Pose(Result=Component To Local(ComponentPose=@n9)) {layerGroup=None}
   Get bIsInFrontEnd()
  ]}
 MasalaRail_StateMachine {
  entry=Idle
  path=AnimGraph.AnimGraphNode_StateMachine_0.MasalaRail_StateMachine
  ids=id
  nodes [
   AnimStateNode_1 = Elimination() {is=title AnimStateNode}
   AnimStateNode_3 = Take Damage() {
    is=title AnimStateNode
    stateEntered {guid=#1}
    stateFullyBlended {guid=#2}
    stateLeft {guid=#3}
   }
   AnimStateNode_4 = Chest Open() {is=title AnimStateNode}
   AnimStateNode_8 = Idle() {
    stateEntered {guid=#4}
    stateFullyBlended {guid=#5}
    stateLeft {guid=#6}
   }
   AnimStateTransitionNode_0 = AnimStateTransitionNode()
   AnimStateTransitionNode_1 = AnimStateTransitionNode() {bAutomaticRuleBasedOnSequencePlayerInState=true}
   AnimStateTransitionNode_10 = AnimStateTransitionNode() {bAutomaticRuleBasedOnSequencePlayerInState=true, crossfadeDuration=0}
   AnimStateTransitionNode_12 = AnimStateTransitionNode() {crossfadeDuration=0}
   AnimStateTransitionNode_5 = AnimStateTransitionNode() {crossfadeDuration=0}
   AnimStateTransitionNode_6 = AnimStateTransitionNode() {bAutomaticRuleBasedOnSequencePlayerInState=true}
  ]
  links [
   {
    dst=Idle
    rule=AnimGraph.AnimGraphNode_StateMachine_0.MasalaRail_StateMachine.AnimStateTransitionNode_10.Transition
    src=Chest Open
    via=AnimStateTransitionNode_10
   }
   {
    dst=Idle
    rule=AnimGraph.AnimGraphNode_StateMachine_0.MasalaRail_StateMachine.AnimStateTransitionNode_1.AnimationTransitionGraph_1
    src=Elimination
    via=AnimStateTransitionNode_1
   }
   {
    dst=Chest Open
    rule=AnimGraph.AnimGraphNode_StateMachine_0.MasalaRail_StateMachine.AnimStateTransitionNode_12.Transition
    src=Idle
    via=AnimStateTransitionNode_12
   }
   {
    dst=Elimination
    rule=AnimGraph.AnimGraphNode_StateMachine_0.MasalaRail_StateMachine.AnimStateTransitionNode_0.AnimationTransitionGraph_2
    src=Idle
    via=AnimStateTransitionNode_0
   }
   {
    dst=Take Damage
    rule=AnimGraph.AnimGraphNode_StateMachine_0.MasalaRail_StateMachine.AnimStateTransitionNode_5.Transition
    src=Idle
    via=AnimStateTransitionNode_5
   }
   {
    dst=Idle
    rule=AnimGraph.AnimGraphNode_StateMachine_0.MasalaRail_StateMachine.AnimStateTransitionNode_6.Transition
    src=Take Damage
    via=AnimStateTransitionNode_6
   }
  ]
 }
 Elimination {
  path=AnimGraph.AnimGraphNode_StateMachine_0.MasalaRail_StateMachine.AnimStateNode_1.Elimination
  nodes [Output Animation Pose(Result=Sequence Player 'MasalaRail_Elim_PACK'() {bLoopAnimation=false, sequence=/BRCosmetics/Accessories/FORT_Backpacks/Backpack_F_MED_MasalaRail/Animations/MasalaRail_Elim_PACK})]
 }
 Take Damage {
  path=AnimGraph.AnimGraphNode_StateMachine_0.MasalaRail_StateMachine.AnimStateNode_3.Take Damage
  nodes [
   n1 = Sequence Player 'MasalaRail_TakeDamage_PACK'() {bLoopAnimation=false, sequence=/BRCosmetics/Accessories/FORT_Backpacks/Backpack_F_MED_MasalaRail/Animations/MasalaRail_TakeDamage_PACK}
   Output Animation Pose(Result=@n1)
  ]
 }
 Chest Open {
  path=AnimGraph.AnimGraphNode_StateMachine_0.MasalaRail_StateMachine.AnimStateNode_4.Chest Open
  nodes [
   n1 = Sequence Player 'MasalaRail_OpenChest_PACK'() {bLoopAnimation=false, sequence=/BRCosmetics/Accessories/FORT_Backpacks/Backpack_F_MED_MasalaRail/Animations/MasalaRail_OpenChest_PACK}
   Output Animation Pose(Result=@n1)
  ]
 }
 Idle {
  path=AnimGraph.AnimGraphNode_StateMachine_0.MasalaRail_StateMachine.AnimStateNode_8.Idle
  nodes [
   Sequence Player 'BasketSuburb_Idle_PACK'() {sequence=/BRCosmetics/Accessories/FORT_Backpacks/Backpack_M_MED_BasketSuburb/Animation/BasketSuburb_Idle_PACK}
   Output Animation Pose(Result=Sequence Player 'MasalaRail_Idle_PACK'() {sequence=/BRCosmetics/Accessories/FORT_Backpacks/Backpack_F_MED_MasalaRail/Animations/MasalaRail_Idle_PACK})
  ]
 }
 AnimationTransitionGraph_2 {
  path=AnimGraph.AnimGraphNode_StateMachine_0.MasalaRail_StateMachine.AnimStateTransitionNode_0.AnimationTransitionGraph_2
  nodes [Result(bCanEnterTransition=Get bEnemyPlayerEliminated() -> (bEnemyPlayerEliminated:bool))]
 }
 AnimationTransitionGraph_1 {path=AnimGraph.AnimGraphNode_StateMachine_0.MasalaRail_StateMachine.AnimStateTransitionNode_1.AnimationTransitionGraph_1, nodes [Result()]}
 Transition {path=AnimGraph.AnimGraphNode_StateMachine_0.MasalaRail_StateMachine.AnimStateTransitionNode_10.Transition, nodes [Result()]}
 Transition {
  path=AnimGraph.AnimGraphNode_StateMachine_0.MasalaRail_StateMachine.AnimStateTransitionNode_12.Transition
  nodes [Result(bCanEnterTransition=Get bChestOpened() -> (bChestOpened:bool))]
 }
 Transition {
  path=AnimGraph.AnimGraphNode_StateMachine_0.MasalaRail_StateMachine.AnimStateTransitionNode_5.Transition
  nodes [Result(bCanEnterTransition=Get Reactive Anim Damage() -> (Reactive Anim Damage:bool))]
 }
 Transition {path=AnimGraph.AnimGraphNode_StateMachine_0.MasalaRail_StateMachine.AnimStateTransitionNode_6.Transition, nodes [Result()]}
 FE_Reactivity {
  entry=Idle
  path=AnimGraph.AnimGraphNode_StateMachine_1.FE_Reactivity
  ids=id
  nodes [
   AnimStateNode_0 = Idle()
   AnimStateNode_2 = FE_React() {is=title AnimStateNode, stateEntered {guid=#7, notifyName=Entered FE_Reactive_Anim}}
   AnimStateTransitionNode_8 = AnimStateTransitionNode()
   AnimStateTransitionNode_9 = AnimStateTransitionNode() {bAutomaticRuleBasedOnSequencePlayerInState=true}
  ]
  links [
   {
    dst=Idle
    rule=AnimGraph.AnimGraphNode_StateMachine_1.FE_Reactivity.AnimStateTransitionNode_9.Transition
    src=FE_React
    via=AnimStateTransitionNode_9
   }
   {
    dst=FE_React
    rule=AnimGraph.AnimGraphNode_StateMachine_1.FE_Reactivity.AnimStateTransitionNode_8.Transition
    src=Idle
    via=AnimStateTransitionNode_8
   }
  ]
 }
 Idle {
  path=AnimGraph.AnimGraphNode_StateMachine_1.FE_Reactivity.AnimStateNode_0.Idle
  nodes [
   Sequence Player 'BasketSuburb_Idle_PACK'() {bLoopAnimation=false, sequence=&BasketSuburb_Idle_PACK}
   Output Animation Pose(Result=Sequence Player 'MasalaRail_Idle_PACK'() {sequence=&MasalaRail_Idle_PACK})
  ]
 }
 FE_React {
  path=AnimGraph.AnimGraphNode_StateMachine_1.FE_Reactivity.AnimStateNode_2.FE_React
  nodes [
   Sequence Player 'MasalaRail_OpenChest_PACK'() {bLoopAnimation=false, sequence=&MasalaRail_OpenChest_PACK}
   Output Animation Pose(Result=Sequence Player 'MasalaRail_Elim_PACK'() {bLoopAnimation=false, sequence=&MasalaRail_Elim_PACK})
  ]
 }
 Transition {
  path=AnimGraph.AnimGraphNode_StateMachine_1.FE_Reactivity.AnimStateTransitionNode_8.Transition
  nodes [Result(bCanEnterTransition=Get Play Reactive FE Anim() -> (Play Reactive FE Anim:bool))]
 }
 Transition {path=AnimGraph.AnimGraphNode_StateMachine_1.FE_Reactivity.AnimStateTransitionNode_9.Transition, nodes [Result()]}
 EventGraph {nodes [
   Event BlueprintUpdateAnimation()
   > n1 = Sequence!(execute) -> (then_0, then_1, then_2:Exec, then_3:Exec)
   > n2 = Branch(Condition=OR Boolean!(A=Get bIsSkydiveDiveMode(), B=Get bIsSkydiveFloating(), C:bool=Get bIsParachuteOpened()) -> (ReturnValue))
   > Set Alpha_skydiveMode(Alpha_skydiveMode:double=1) -> (Output_Get:double)
   Set Alpha_skydiveMode(execute=@n2.else, Alpha_skydiveMode=0)
   n3 = BreakTransform(InTransform=GetDeltaTransformFromRefPose(self=GetOwnerSkeletalMeshComponent(), BoneName=spine_05, BaseName=spine_01))
   Set Parachute _correction(execute=@n1.then_1, Parachute _correction:double=MapRangeClamped(Value=BreakRotator(InRot=@n3.Rotation).Yaw, InRangeA=-25, InRangeB=-65)) -> (Output_Get:double)
   Event OnCharacterCustomizationLoaded()
   > Cast To FortPlayerPawn(Object=TryGetPawnOwner())
   > n4 = Branch()
   > Do Once(Reset=@n6.then)
   > n5 = Delay(Duration=3)
   Set Play Reactive FE Anim(execute=[@n7.then, @n5.then], Play Reactive FE Anim:bool=true) -> (Output_Get:bool)
   n6 = Set Play Reactive FE Anim(execute=@n4.else, Play Reactive FE Anim=false)
   AnimNotify_FE_ReactiveAnim() {is=title K2Node_Event}
   > Set Play Reactive FE Anim(Play Reactive FE Anim=false)
   > n7 = Delay(Duration=10)
   Event BlueprintPostEvaluateAnimation() {is=title K2Node_Event}
   > Set bEnemyPlayerEliminated(bEnemyPlayerEliminated:bool=false) -> (Output_Get:bool)
   > Set Play Reactive FE Anim(Play Reactive FE Anim=false)
   n8 = Event BlueprintProcessAnimInputEvent()
   > n9 = Sequence!(execute) -> (then_0, then_1, then_2)
   > Branch(Condition=IsNamed(self=@n8.AnimInputEvent, InEventName=aie.ChestOpened))
   > Set bChestOpened(bChestOpened:bool=true) -> (Output_Get:bool)
   > RetriggerableDelay(Duration=0.2)
   > Set bChestOpened(bChestOpened=false)
   Branch(
    execute=@n9.then_1
    Condition=OR Boolean!(
     A=IsNamed(self=@n8.AnimInputEvent, InEventName=aie.PlayerVitality.ShieldsDecreased)
     B=IsNamed(self=@n8.AnimInputEvent, InEventName=aie.PlayerVitality.HealthDecreased)
    ) -> (ReturnValue)
   )
   > Set Reactive Anim Damage(Reactive Anim Damage:bool=true) -> (Output_Get:bool)
   > RetriggerableDelay(Duration=0.2)
   > Set Reactive Anim Damage(Reactive Anim Damage=false)
   Branch(execute=@n9.then_2, Condition=IsNamed(self=@n8.AnimInputEvent, InEventName=aie.EnemyPlayerEliminated))
   > Set bEnemyPlayerEliminated(bEnemyPlayerEliminated=true)
   > Delay(Duration=0.2)
   > Set bEnemyPlayerEliminated(bEnemyPlayerEliminated=false)
   TryGetPawnOwner()
  ]}
]
variables [
 Alpha_skydiveMode {type=number}
 bChestOpened {type=boolean}
 bEnemyPlayerEliminated {type=boolean}
 FrontEnd {type=boolean}
 isHappy {type=boolean}
 Parachute _correction {type=number}
 Play Reactive FE Anim {type=boolean}
 PlayerPawn {type=FortPlayerPawn}
 Reactive Anim Damage {type=boolean}
 ReactiveAnimation {type=boolean}
 Reactivity Enabled {type=boolean}
]
